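# Sound cue for the drop sound of the metal bow, classed under Fort_Loot_Collect.
# It plays a single Bone bow drop wave through a Modulator that randomizes pitch, reusing the Bone asset in place of a metal-specific one.
# The cue slightly lowers overall pitch and volume, uses weapon-drop attenuation, and is limited by global and local bow concurrency rules.
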SoundCue /PrimalGameplay/Sounds/Weapons/Bows/Metal/Bow_Metal_Drop_Cue
root=Modulator_0
nodes [
 {id=Modulator_0, children [WavePlayer_2], properties {pitchMax=1.2}}
 {id=WavePlayer_2, properties {soundWaveAssetPtr=/PrimalGameplay/Sounds/Weapons/Bows/Bone/Bow_Bone_Drop_02}}
]
properties {
 attenuationSettings=&Fort_Weapon_Drop_Att
 maxDistance=3000
 pitchMultiplier=0.9
 soundClassObject=&Fort_Loot_Collect
 volumeMultiplier=0.8
 concurrencySet [/Game/Sounds/Concurrency_Rules/Bow_Global_Conc, /Game/Sounds/Concurrency_Rules/Bow_Local_Conc]
}
