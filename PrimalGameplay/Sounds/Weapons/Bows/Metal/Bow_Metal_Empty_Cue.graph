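# Sound cue for the metal bow's empty sound, with no sound nodes or root, so it plays no audio as shown.
# It carries only playback settings: the weapon foley sound class, a reduced volume multiplier and an attenuation falloff override.
# Global and local bow concurrency rules are assigned to it.
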SoundCue /PrimalGameplay/Sounds/Weapons/Bows/Metal/Bow_Metal_Empty_Cue
root=null
nodes []
properties {
 maxDistance=2097152
 soundClassObject=&Fort_WeaponFoley
 volumeMultiplier=0.251
 attenuationOverrides {falloffDistance=1600}
 concurrencySet [/Game/Sounds/Concurrency_Rules/Bow_Global_Conc, /Game/Sounds/Concurrency_Rules/Bow_Local_Conc]
}
