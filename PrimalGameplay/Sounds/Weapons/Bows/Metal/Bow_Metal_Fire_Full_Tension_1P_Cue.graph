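# First-person sound cue for firing a metal bow at full tension.
# A mixer layers two randomized branches: one of three full-tension fire waves and one of five whoosh waves, each through a modulator, with the whoosh branch's pitch raised by a random amount.
# It routes through a first-person ranged-fire sound class, first-person weapon attenuation and a gunfire concurrency set.
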SoundCue /PrimalGameplay/Sounds/Weapons/Bows/Metal/Bow_Metal_Fire_Full_Tension_1P_Cue
root=Mixer_1
nodes [
 {id=Mixer_1, children [Modulator_1,Modulator_2], properties {inputVolume [0.6,0.5]}}
 {id=Modulator_1, children [Random_0]}
 {id=Modulator_2, children [Random_2], properties {pitchMax=1.25, pitchMin=1}}
 {id=Random_0, children [WavePlayer_0,WavePlayer_1,WavePlayer_2], properties {weights [1,1,1]}}
 {id=Random_2, children [WavePlayer_9, WavePlayer_10, WavePlayer_11, WavePlayer_12, WavePlayer_13], properties {weights [1,1,1,1,1]}}
 {id=WavePlayer_0, properties {soundWaveAssetPtr=&/Bow_Metal_Fire_Full_Tension_01}}
 {id=WavePlayer_1, properties {soundWaveAssetPtr=&/Bow_Metal_Fire_Full_Tension_02}}
 {id=WavePlayer_10, properties {soundWaveAssetPtr=&/Bow_Metal_Fire_Whoosh_02}}
 {id=WavePlayer_11, properties {soundWaveAssetPtr=&/Bow_Metal_Fire_Whoosh_03}}
 {id=WavePlayer_12, properties {soundWaveAssetPtr=&/Bow_Metal_Fire_Whoosh_04}}
 {id=WavePlayer_13, properties {soundWaveAssetPtr=&/Bow_Metal_Fire_Whoosh_05}}
 {id=WavePlayer_2, properties {soundWaveAssetPtr=&/Bow_Metal_Fire_Full_Tension_03}}
 {id=WavePlayer_9, properties {soundWaveAssetPtr=&/Bow_Metal_Fire_Whoosh_01}}
]
properties {
 attenuationSettings=&Fort_Weapon_Projectile_FirstPerson_Athena_Att
 maxDistance=2097152
 soundClassObject=&Fort_WeaponRanged_Fire_1P_PSM
 volumeMultiplier=1
 attenuationOverrides {falloffDistance=5000, attenuationShapeExtents {x=1000}}
 concurrencySet [&Weapon_Gunfire_FirstPerson_Conc]
}
