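# Sound cue for a metal bow firing at full tension, mixing a close third-person layer with a distant layer.
# The close layer blends random full-tension shots with quieter random whooshes, whose pitch is modulated upward.
# The distant layer randomly picks from the default bow's charged distant samples rather than metal-specific ones.
# Each layer has its own sound class and attenuation, and global and local bow concurrency rules govern the cue.
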SoundCue /PrimalGameplay/Sounds/Weapons/Bows/Metal/Bow_Metal_Fire_Full_Tension_Cue
root=Mixer_3
nodes [
 {id=Attenuation_0, children [Mixer_2], properties {attenuationSettings=/Game/Sounds/Attenuation/Fort_Weapon_Projectile_Close_Bow_Athena_Att}}
 {
  id=Attenuation_2
  children [Modulator_8]
  properties {attenuationSettings=/Game/Sounds/Attenuation/Fort_Weapon_Projectile_Distant_Bow_Athena_Att}
 }
 {id=Mixer_2, children [Modulator_3,Modulator_4], properties {inputVolume [1,0.3]}}
 {id=Mixer_3, children [SoundClass_0,SoundClass_4], properties {inputVolume [0.65,0.7]}}
 {id=Modulator_3, children [Random_4]}
 {id=Modulator_4, children [Random_5], properties {pitchMax=1.2, pitchMin=1}}
 {id=Modulator_8, children [Random_0]}
 {id=Random_0, children [WavePlayer_0,WavePlayer_1,WavePlayer_2], properties {weights [1,1,1]}}
 {id=Random_4, children [WavePlayer_15, WavePlayer_16, WavePlayer_17], properties {weights [1,1,1]}}
 {id=Random_5, children [WavePlayer_18, WavePlayer_19, WavePlayer_20, WavePlayer_24, WavePlayer_25], properties {weights [1,1,1,1,1]}}
 {
  id=SoundClass_0
  children [Attenuation_0]
  properties {soundClassOverride=&Fort_WeaponRanged_Fire_3P_PSM}
 }
 {
  id=SoundClass_4
  children [Attenuation_2]
  properties {soundClassOverride=&Fort_WeaponRanged_Fire_Distant}
 }
 {id=WavePlayer_0, properties {soundWaveAssetPtr=&/Bow_Default_Fire_3P_Charged_Distant_01}}
 {id=WavePlayer_1, properties {soundWaveAssetPtr=&/Bow_Default_Fire_3P_Charged_Distant_02}}
 {id=WavePlayer_15, properties {soundWaveAssetPtr=&/Bow_Metal_Fire_Full_Tension_01}}
 {id=WavePlayer_16, properties {soundWaveAssetPtr=&/Bow_Metal_Fire_Full_Tension_02}}
 {id=WavePlayer_17, properties {soundWaveAssetPtr=&/Bow_Metal_Fire_Full_Tension_03}}
 {id=WavePlayer_18, properties {soundWaveAssetPtr=&/Bow_Metal_Fire_Whoosh_01}}
 {id=WavePlayer_19, properties {soundWaveAssetPtr=&/Bow_Metal_Fire_Whoosh_02}}
 {id=WavePlayer_2, properties {soundWaveAssetPtr=&/Bow_Default_Fire_3P_Charged_Distant_03}}
 {id=WavePlayer_20, properties {soundWaveAssetPtr=&/Bow_Metal_Fire_Whoosh_03}}
 {id=WavePlayer_24, properties {soundWaveAssetPtr=&/Bow_Metal_Fire_Whoosh_04}}
 {id=WavePlayer_25, properties {soundWaveAssetPtr=&/Bow_Metal_Fire_Whoosh_05}}
]
properties {
 maxDistance=27500
 volumeMultiplier=1
 attenuationOverrides {falloffDistance=5000, attenuationShapeExtents {x=1000}}
 concurrencySet [/Game/Sounds/Concurrency_Rules/Bow_Global_Conc, /Game/Sounds/Concurrency_Rules/Bow_Local_Conc]
}
