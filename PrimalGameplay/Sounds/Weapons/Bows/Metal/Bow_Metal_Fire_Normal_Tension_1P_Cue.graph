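# First-person sound cue for firing a metal bow's normal tension shot.
# A mixer layers one of three random fire-tension waves with one of six random distant passby waves, the latter delayed slightly.
# Each layer runs through its own modulator, and the mixer weights the passby layer a little higher than the fire layer.
# It uses a first-person projectile attenuation, a ranged-weapon fire sound class and first-person gunfire concurrency.
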SoundCue /PrimalGameplay/Sounds/Weapons/Bows/Metal/Bow_Metal_Fire_Normal_Tension_1P_Cue
root=Mixer_1
nodes [
 {id=Delay_0, children [Modulator_1], properties {delayMax=0.1, delayMin=0.1}}
 {id=Mixer_1, children [Modulator_2,Delay_0], properties {inputVolume [0.8,1.1]}}
 {id=Modulator_1, children [Random_1]}
 {id=Modulator_2, children [Random_0]}
 {id=Random_0, children [WavePlayer_3,WavePlayer_4,WavePlayer_5], properties {weights [1,1,1]}}
 {
  id=Random_1
  children [WavePlayer_6, WavePlayer_7, WavePlayer_8, WavePlayer_9, WavePlayer_10, WavePlayer_11]
  properties {weights [1,1,1,1,1,1]}
 }
 {id=WavePlayer_10, properties {soundWaveAssetPtr=&/Bow_Metal_Passby_Default_Distant_04}}
 {id=WavePlayer_11, properties {soundWaveAssetPtr=&/Bow_Metal_Passby_Default_Distant_05}}
 {id=WavePlayer_3, properties {soundWaveAssetPtr=&/Bow_Metal_Fire_Normal_Tension_01}}
 {id=WavePlayer_4, properties {soundWaveAssetPtr=&/Bow_Metal_Fire_Normal_Tension_02}}
 {id=WavePlayer_5, properties {soundWaveAssetPtr=&/Bow_Metal_Fire_Normal_Tension_03}}
 {id=WavePlayer_6, properties {soundWaveAssetPtr=&/Bow_Metal_Passby_Default_Distant_06}}
 {id=WavePlayer_7, properties {soundWaveAssetPtr=&/Bow_Metal_Passby_Default_Distant_01}}
 {id=WavePlayer_8, properties {soundWaveAssetPtr=&/Bow_Metal_Passby_Default_Distant_02}}
 {id=WavePlayer_9, properties {soundWaveAssetPtr=&/Bow_Metal_Passby_Default_Distant_03}}
]
properties {
 attenuationSettings=&Fort_Weapon_Projectile_FirstPerson_Athena_Att
 maxDistance=2097152
 soundClassObject=&Fort_WeaponRanged_Fire_1P_PSM
 volumeMultiplier=0.6
 attenuationOverrides {falloffDistance=5000, attenuationShapeExtents {x=1000}}
 concurrencySet [&Weapon_Gunfire_FirstPerson_Conc]
}
