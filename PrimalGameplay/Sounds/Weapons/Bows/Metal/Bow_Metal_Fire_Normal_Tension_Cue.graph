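# Sound Cue for a metal bow firing at normal tension, mixing close and distant layers, each with its own attenuation and sound class override.
# The close layer, on the third-person fire class, mixes a random tension shot with a briefly delayed random metal passby.
# The distant layer randomly picks a default-bow distant shot.
# Modulators vary each random pick, and global and local bow concurrency rules limit playback.
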SoundCue /PrimalGameplay/Sounds/Weapons/Bows/Metal/Bow_Metal_Fire_Normal_Tension_Cue
root=Mixer_3
nodes [
 {id=Attenuation_0, children [Mixer_4], properties {attenuationSettings=/Game/Sounds/Attenuation/Fort_Weapon_Projectile_Close_Bow_Athena_Att}}
 {
  id=Attenuation_2
  children [Modulator_8]
  properties {attenuationSettings=/Game/Sounds/Attenuation/Fort_Weapon_Projectile_Distant_Bow_Athena_Att}
 }
 {id=Delay_0, children [Modulator_7], properties {delayMax=0.1, delayMin=0.1}}
 {id=Mixer_3, children [SoundClass_0,SoundClass_4], properties {inputVolume [0.65,0.7]}}
 {id=Mixer_4, children [Modulator_6,Delay_0], properties {inputVolume [0.8,1.1]}}
 {id=Modulator_6, children [Random_5]}
 {id=Modulator_7, children [Random_7]}
 {id=Modulator_8, children [Random_0]}
 {id=Random_0, children [WavePlayer_3,WavePlayer_4,WavePlayer_5], properties {weights [1,1,1]}}
 {id=Random_5, children [WavePlayer_19, WavePlayer_20, WavePlayer_24], properties {weights [1,1,1]}}
 {
  id=Random_7
  children [WavePlayer_6, WavePlayer_25, WavePlayer_8, WavePlayer_26, WavePlayer_27, WavePlayer_28]
  properties {weights [1,1,1,1,1,1]}
 }
 {
  id=SoundClass_0
  children [Attenuation_0]
  properties {soundClassOverride=&Fort_WeaponRanged_Fire_3P_PSM}
 }
 {
  id=SoundClass_4
  children [Attenuation_2]
  properties {soundClassOverride=&Fort_WeaponRanged_Fire_Distant}
 }
 {id=WavePlayer_19, properties {soundWaveAssetPtr=&/Bow_Metal_Fire_Normal_Tension_01}}
 {id=WavePlayer_20, properties {soundWaveAssetPtr=&/Bow_Metal_Fire_Normal_Tension_02}}
 {id=WavePlayer_24, properties {soundWaveAssetPtr=&/Bow_Metal_Fire_Normal_Tension_03}}
 {id=WavePlayer_25, properties {soundWaveAssetPtr=&/Bow_Metal_Passby_Default_Distant_01}}
 {id=WavePlayer_26, properties {soundWaveAssetPtr=&/Bow_Metal_Passby_Default_Distant_03}}
 {id=WavePlayer_27, properties {soundWaveAssetPtr=&/Bow_Metal_Passby_Default_Distant_04}}
 {id=WavePlayer_28, properties {soundWaveAssetPtr=&/Bow_Metal_Passby_Default_Distant_05}}
 {id=WavePlayer_3, properties {soundWaveAssetPtr=&/Bow_Default_Fire_3P_Normal_Distant_03}}
 {id=WavePlayer_4, properties {soundWaveAssetPtr=&/Bow_Default_Fire_3P_Normal_Distant_01}}
 {id=WavePlayer_5, properties {soundWaveAssetPtr=&/Bow_Default_Fire_3P_Normal_Distant_02}}
 {id=WavePlayer_6, properties {soundWaveAssetPtr=&/Bow_Metal_Passby_Default_Distant_06}}
 {id=WavePlayer_8, properties {soundWaveAssetPtr=&/Bow_Metal_Passby_Default_Distant_02}}
]
properties {
 maxDistance=27500
 volumeMultiplier=1
 attenuationOverrides {falloffDistance=5000, attenuationShapeExtents {x=1000}}
 concurrencySet [/Game/Sounds/Concurrency_Rules/Bow_Global_Conc, /Game/Sounds/Concurrency_Rules/Bow_Local_Conc]
}
