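# Sound cue that plays the full-tension confirmation sound for the metal bow.
# A Mixer layers two waves from the default bow's full-tension confirm set at equal volume, so both play together.
# It uses weapon foley sound class, weapon reload attenuation and global and local bow concurrency rules, at roughly half volume.
# The debug flag is left on.
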SoundCue /PrimalGameplay/Sounds/Weapons/Bows/Metal/Bow_Metal_Full_Tension_Confirm_Cue
root=Mixer_1
nodes [
 {id=Mixer_1, children [WavePlayer_1,WavePlayer_0], properties {inputVolume [1,1]}}
 {id=WavePlayer_0, properties {soundWaveAssetPtr=/PrimalGameplay/Sounds/Weapons/Bows/Default/Bow_Default_FullTension_Confirm_02}}
 {id=WavePlayer_1, properties {soundWaveAssetPtr=/PrimalGameplay/Sounds/Weapons/Bows/Default/Bow_Default_FullTension_Confirm_01}}
]
properties {
 attenuationSettings=&Fort_Weapon_Reload_Att
 bDebug=true
 maxDistance=2250
 soundClassObject=&Fort_WeaponFoley
 volumeMultiplier=0.501
 concurrencySet [/Game/Sounds/Concurrency_Rules/Bow_Global_Conc, /Game/Sounds/Concurrency_Rules/Bow_Local_Conc]
}
