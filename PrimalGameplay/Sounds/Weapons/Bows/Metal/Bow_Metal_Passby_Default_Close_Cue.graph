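# Sound cue for the close-range pass-by sound of the default metal bow.
# A random node picks one of six equally weighted wave variations and plays it at slightly reduced volume.
# It routes through the ranged weapon sound class, references the compound bow whoosh-by attenuation asset, and is capped by global and local bow concurrency rules.
# The debug flag is enabled on the cue.
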SoundCue /PrimalGameplay/Sounds/Weapons/Bows/Metal/Bow_Metal_Passby_Default_Close_Cue
root=Random_1
nodes [
 {id=Random_1, children [WavePlayer_0,WavePlayer_5,WavePlayer_6,WavePlayer_7,WavePlayer_8,WavePlayer_9], properties {weights [1,1,1,1,1,1]}}
 {id=WavePlayer_0, properties {soundWaveAssetPtr=&/Bow_Metal_Passby_Default_Close_01}}
 {id=WavePlayer_5, properties {soundWaveAssetPtr=&/Bow_Metal_Passby_Default_Close_02}}
 {id=WavePlayer_6, properties {soundWaveAssetPtr=&/Bow_Metal_Passby_Default_Close_03}}
 {id=WavePlayer_7, properties {soundWaveAssetPtr=&/Bow_Metal_Passby_Default_Close_04}}
 {id=WavePlayer_8, properties {soundWaveAssetPtr=&/Bow_Metal_Passby_Default_Close_05}}
 {id=WavePlayer_9, properties {soundWaveAssetPtr=&/Bow_Metal_Passby_Default_Close_06}}
]
properties {
 attenuationSettings=/Game/Sounds/Fort_Weapons/Explosive_Bow/Weapon_Compound_Bow_Whoosh_By_Att
 bDebug=true
 maxDistance=2097152
 soundClassObject=&Fort_WeaponRanged
 volumeMultiplier=0.7
 attenuationOverrides {
  bAttenuate=false
  bEnableReverbSend=false
  bUseComplexCollisionForOcclusion=true
  falloffDistance=0
  occlusionVolumeAttenuation=0
  stereoSpread=0
  attenuationShapeExtents {x=50000}
 }
 concurrencySet [/Game/Sounds/Concurrency_Rules/Bow_Global_Conc, /Game/Sounds/Concurrency_Rules/Bow_Local_Conc]
}
