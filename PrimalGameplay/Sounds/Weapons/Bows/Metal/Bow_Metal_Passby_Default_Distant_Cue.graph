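# Sound cue for the distant pass-by sound of a metal bow, classed as a ranged weapon sound.
# A single Random node picks among six equally weighted wave players, each holding one distant pass-by variation.
# It references a compound bow whoosh-by attenuation asset and uses global and local bow concurrency rules.
# The cue's debug flag is left on.
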SoundCue /PrimalGameplay/Sounds/Weapons/Bows/Metal/Bow_Metal_Passby_Default_Distant_Cue
root=Random_0
nodes [
 {id=Random_0, children [WavePlayer_1,WavePlayer_2,WavePlayer_3,WavePlayer_4,WavePlayer_8,WavePlayer_9], properties {weights [1,1,1,1,1,1]}}
 {id=WavePlayer_1, properties {soundWaveAssetPtr=&/Bow_Metal_Passby_Default_Distant_06}}
 {id=WavePlayer_2, properties {soundWaveAssetPtr=&/Bow_Metal_Passby_Default_Distant_01}}
 {id=WavePlayer_3, properties {soundWaveAssetPtr=&/Bow_Metal_Passby_Default_Distant_02}}
 {id=WavePlayer_4, properties {soundWaveAssetPtr=&/Bow_Metal_Passby_Default_Distant_03}}
 {id=WavePlayer_8, properties {soundWaveAssetPtr=&/Bow_Metal_Passby_Default_Distant_04}}
 {id=WavePlayer_9, properties {soundWaveAssetPtr=&/Bow_Metal_Passby_Default_Distant_05}}
]
properties {
 attenuationSettings=/Game/Sounds/Fort_Weapons/Explosive_Bow/Weapon_Compound_Bow_Whoosh_By_Att
 bDebug=true
 maxDistance=2097152
 soundClassObject=&Fort_WeaponRanged
 volumeMultiplier=1
 attenuationOverrides {
  bAttenuate=false
  bEnableReverbSend=false
  bUseComplexCollisionForOcclusion=true
  falloffDistance=1000
  occlusionVolumeAttenuation=0
  stereoSpread=0
  attenuationShapeExtents {x=50000}
 }
 concurrencySet [/Game/Sounds/Concurrency_Rules/Bow_Global_Conc, /Game/Sounds/Concurrency_Rules/Bow_Local_Conc]
}
