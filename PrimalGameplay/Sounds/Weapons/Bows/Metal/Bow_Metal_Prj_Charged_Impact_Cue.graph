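# Sound cue for the impact of a charged projectile from the metal bow.
# A random node picks one of five equally weighted impact waves and feeds it through a modulator at the root.
# It plays in the impacts sound class with boosted volume, using arrow impact attenuation and both global and local bow concurrency rules.
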SoundCue /PrimalGameplay/Sounds/Weapons/Bows/Metal/Bow_Metal_Prj_Charged_Impact_Cue
root=Modulator_3
nodes [
 {id=Modulator_3, children [Random_2]}
 {id=Random_2, children [WavePlayer_7, WavePlayer_8, WavePlayer_14, WavePlayer_15, WavePlayer_16], properties {weights [1,1,1,1,1]}}
 {id=WavePlayer_14, properties {soundWaveAssetPtr=&/Bow_Metal_Prj_Charged_Impact_03}}
 {id=WavePlayer_15, properties {soundWaveAssetPtr=&/Bow_Metal_Prj_Charged_Impact_04}}
 {id=WavePlayer_16, properties {soundWaveAssetPtr=&/Bow_Metal_Prj_Charged_Impact_05}}
 {id=WavePlayer_7, properties {soundWaveAssetPtr=&/Bow_Metal_Prj_Charged_Impact_01}}
 {id=WavePlayer_8, properties {soundWaveAssetPtr=&/Bow_Metal_Prj_Charged_Impact_02}}
]
properties {
 attenuationSettings=/Game/Sounds/Attenuation/Fort_Arrow_Impact_Athena_Att
 maxDistance=6000
 soundClassObject=&Fort_Impacts
 volumeMultiplier=2.016
 concurrencySet [/Game/Sounds/Concurrency_Rules/Bow_Global_Conc, /Game/Sounds/Concurrency_Rules/Bow_Local_Conc]
}
